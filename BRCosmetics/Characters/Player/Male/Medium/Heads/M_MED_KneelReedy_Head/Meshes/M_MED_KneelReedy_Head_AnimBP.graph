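# Animation Blueprint for the KneelReedy male head that adjusts material temporal responsiveness for its flipbook face.
# After the parent update, it compares the mouth and eye curve indices with cached values.
# It also derives a flip index from which side of the faceAttach socket's forward axis the camera is on.
# Any change raises the mesh's Temporal Responsiveness parameter from 0.5 to 1 for that update, then the values are cached.
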AnimBlueprint /BRCosmetics/Characters/Player/Male/Medium/Heads/M_MED_KneelReedy_Head/Meshes/M_MED_KneelReedy_Head_AnimBP
target_skeleton=/Game/Characters/Player/Common/Fortnite_Base_Head/Fortnite_Base_Head_Export_Skeleton
graphs [EventGraph {nodes [
   n1 = Event BlueprintUpdateAnimation()
   > Parent: BlueprintUpdateAnimation!(execute, DeltaTimeX:float=@n1.DeltaTimeX) -> (then) {is=title K2Node_CallParentFunction}
   n2 = GetPartSkeletalMeshComponent()
   n3 = GetAnimInstance(self=@n2)
   n4 = GetCurveValue(self=@n3, CurveName=flipbook_face_mouth_index)
   n5 = "Not Equal (Float)"(A=@n4, B=Get Mouth Index() -> (Mouth Index:float))
   n6 = GetCurveValue(self=@n3, CurveName=flipbook_face_R_eye_index)
   n7 = "Not Equal (Float)"(A=@n6, B=Get Right Eye Index() -> (Right Eye Index:float))
   n8 = GetCurveValue(self=@n3, CurveName=flipbook_face_L_eye_index)
   n9 = "Not Equal (Float)"(A=@n8, B=Get Left Eye Index() -> (Left Eye Index:float))
   n10 = GetSocketTransform!(self=@n2, InSocketName=faceAttach, TransformSpace=RTS_World) -> (ReturnValue_Location:Vector, ReturnValue_Rotation:Rotator, ReturnValue_Scale:Vector)
   n11 = GetPlayerCameraManager!(PlayerIndex:int=0) -> (ReturnValue:Player Camera Manager Object Reference)
   n12 = Normalize!(
    A:Vector=vector - vector(A=GetCameraLocation!(self:Player Camera Manager Object Reference=@n11) -> (ReturnValue:Vector), B=@n10)
    Tolerance:float=0.0001
   ) -> (ReturnValue:Vector) {is=title K2Node_CallFunction}
   n13 = GetForwardVector!(InRot:Rotator=@n10.ReturnValue_Rotation) -> (ReturnValue:Vector)
   n14 = "Clamp (Float)"(Value="Sign (Float)"!(A:double=Dot Product!(A:Vector=@n12, B:Vector=@n13) -> (ReturnValue:double)) -> (ReturnValue:double), Min=0, Max=1)
   n15 = "Not Equal (Float)"(A=Get Flip Index() -> (Flip Index:double), B=@n14)
   > SetScalarParameterValueOnMaterials!(
    execute
    self:Mesh Component Object Reference=@n2
    ParameterName:Name=Temporal Responsiveness
    ParameterValue:float=SelectFloat(A=1, B=0.5, bPickA=OR Boolean!(A=@n7, B=@n9, C:bool=@n5, D:bool=@n15) -> (ReturnValue))
   ) -> (then)
   > Set Left Eye Index(Left Eye Index:float=@n8) -> (Output_Get:float)
   > Set Right Eye Index(Right Eye Index:float=@n6) -> (Output_Get:float)
   > Set Mouth Index(Mouth Index:float=@n4) -> (Output_Get:float)
   > Set Flip Index(Flip Index:double=@n14) -> (Output_Get:double)
  ]}]
variables [
 Flip Index {type=number}
 Left Eye Index {type=number}
 Mouth Index {type=number}
 Outline Material {type=MaterialInstanceDynamic}
 Right Eye Index {type=number}
]
